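# Editor utility that generates navigation edges on the static meshes selected in the editor.
# GenerateEdgesOnSelectedMeshes gathers the selected StaticMesh assets, checks them out and loops over them, calling GenerateNavigationEdgesOnStaticMesh on each.
# Generation uses the shared GenParam_Default settings for edge size and ground angle.
# It also saves the selected meshes, limited to those left dirty.
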EditorUtilityBlueprint /Game/ArtTools/Environment/EdgeGeneration/AAU_GeneratedEdgesOnStaticMeshes
graphs [
 EventGraph {links [], nodes []}
 GenerateEdgesOnSelectedMeshes {nodes [
   GenerateEdgesOnSelectedMeshes() {is=title K2Node_FunctionEntry}
   > n1 = GetSelectedAssetsOfClass(AssetClass=&StaticMesh)
   > n2 = Set L_AR_SelectedMeshes(L_AR_SelectedMeshes:Array of Object References=@n1) -> (Output_Get:Array of Object References)
   > CheckoutLoadedAssets(AssetsToCheckout=@n2)
   > n3 = For Each Loop(Array:Array of Object References=@n2) -> (Array Element:Object Reference)
   > n4 = Cast To StaticMesh(Object=@n3)
   > GenerateNavigationEdgesOnStaticMesh!(
    execute
    StaticMesh:Static Mesh Object Reference=@n4
    GenerationParam:"Fort Navigation Edges Generation Param Structure (by ref)"=Get GenParam_Default() -> (GenParam_Default:Fort Navigation Edges Generation Param Structure)
   ) -> (then)
   > SaveLoadedAssets(AssetsToSave=Get L_AR_SelectedMeshes() -> (L_AR_SelectedMeshes:Array of Object References), bOnlyIfIsDirty=true)
  ]}
]
variables [GenParam_Default {type=FortNavigationEdgesGenerationParam, default {expectedMaxEdge=20, maxGroundAngle=44, minEdgeSize=5}}]
