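# EUW_EdgeGenerationUtility is an editor utility widget that provides a one-button tool for generating edges on selected meshes during environment art work.
# Its layout is a single Generate Edges button inside a stack and vertical box.
# Clicking the button constructs an AAU_GeneratedEdgesOnStaticMeshes object and calls its GenerateEdgesOnSelectedMeshes function, so the generation logic lives in that class rather than in the widget.
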EditorUtilityWidgetBlueprint /Game/ArtTools/Environment/EdgeGeneration/EUW_EdgeGenerationUtility
graphs [EventGraph {nodes [
   "On Clicked (BTN_GenerateEdges)"() {is=title K2Node_ComponentBoundEvent}
   > n1 = Construct Object from Class(Class=&/AAU_GeneratedEdgesOnStaticMeshes.AAU_GeneratedEdgesOnStaticMeshes_C) -> (ReturnValue:AAU Generated Edges On Static Meshes Object Reference)
   > GenerateEdgesOnSelectedMeshes(self:AAU Generated Edges On Static Meshes Object Reference=@n1)
   Event PreConstruct()
   Event Construct()
   Event Tick()
  ]}]
variables []
widgets {
 info {rootWidgetClass=&StackBox, widgetCount=4}
 widgets [
  StackBox_0 {widgetClassPath=&StackBox, widget=StackBox_0}
  VerticalBox_0 {
   widgetClassPath=&VerticalBox
   parent=StackBox_0
   widget=VerticalBox_0
   properties {bOverride_Cursor=true}
   slot {class=StackBoxSlot}
  }
  BTN_GenerateEdges {
   widgetClassPath=&EditorUtilityButton
   parent=VerticalBox_0
   widget=BTN_GenerateEdges
   slot {class=VerticalBoxSlot, properties {padding=<ltrb 5 5 5 5>}}
  }
  TXT_BTN_LBL {
   widgetClassPath=&UEFN_TextBlock
   parent=BTN_GenerateEdges
   widget=TXT_BTN_LBL
   properties {text=Generate Edges}
   slot {class=ButtonSlot}
  }
 ]
}
